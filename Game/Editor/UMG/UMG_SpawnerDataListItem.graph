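# Editor utility list-row widget, derived from FortAthenaSpawnerDataEditorListItem, that represents one asset in a spawner data editor list.
# A horizontal row holds a name label, an inherited SinglePropertyView for AssetTags, and open and browse icon buttons.
# Both buttons read the row's asset data via Get Asset, either opening the asset through AssetEditorSubsystem or syncing the Content Browser to it.
# A GetText function returns the asset name as text.
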EditorUtilityWidgetBlueprint /Game/Editor/UMG/UMG_SpawnerDataListItem
graphs [
 EventGraph {nodes [
   "On Clicked (BrowseToAssetButton)"() {is=title K2Node_ComponentBoundEvent}
   n1 = Break Asset Data(AssetData=Get Asset() -> (Asset:Asset Data Structure))
   > SyncBrowserToObjects(AssetPaths=Make Array("[0]":String="To String (Name)"(InName=@n1)) -> (Array:Array of Strings))
   "On Clicked (OpenAssetButton)"() {is=title K2Node_ComponentBoundEvent}
   n2 = Get Asset()
   > OpenEditorForAssets!(
    execute
    self:Asset Editor Subsystem Object Reference=Get AssetEditorSubsystem()
    Assets:Array of Object References=Make Array("[0]":Object Reference=GetAsset(InAssetData=@n2)) -> (Array:Array of Object References)
    OpenedMethod:EAssetTypeActivationOpenedMethod Enum=Edit
   ) -> (then, ReturnValue:bool)
  ]}
 GetText {nodes [
   GetText() {is=title K2Node_FunctionEntry}
   n1 = Get Asset()
   > Return Node!(execute, ReturnValue:Text="To Text (Name)"!(InName:Name=Break Asset Data(AssetData=@n1).AssetName) -> (ReturnValue:Text))
  ]}
]
variables []
widgets {
 info {
  inheritedWidgetCount=1
  parentClass=/Script/FortniteEditor.FortAthenaSpawnerDataEditorListItem
  rootWidgetClass=&HorizontalBox
  widgetCount=7
 }
 widgets [
  HorizontalBox_30 {widgetClassPath=&HorizontalBox, widget=HorizontalBox_30}
  AssetName {
   widgetClassPath=&TextBlock
   parent=HorizontalBox_30
   widget=AssetName
   properties {
    text=Text Block
    font {
     bForceMonospaced=false
     bMaterialIsStencil=false
     fontMaterial=null
     fontObject=&Roboto
     letterSpacing=0
     monospacedWidth=1
     size=9
     skewAmount=0
     typefaceFontName=Bold
     outlineSettings {
      bApplyOutlineToDropShadows=false
      bMiteredCorners=false
      bSeparateFillAlpha=false
      outlineMaterial=null
      outlineSize=0
      outlineColor=<rgba 0 0 0 1>
     }
    }
   }
   slot {class=HorizontalBoxSlot, properties {size {sizeRule=Fill, value=1}}}
  }
  AssetTagsView {
   widgetClassPath=&SinglePropertyView
   bInherited=true
   parent=HorizontalBox_30
   widget=AssetTagsView
   properties {accessibleBehavior=Auto, cursor=None, nameOverride=Tags, propertyName=AssetTags}
   slot {class=HorizontalBoxSlot, properties {size {sizeRule=Fill, value=1}}}
  }
  OpenAssetButton {
   widgetClassPath=&EditorUtilityButton
   parent=HorizontalBox_30
   widget=OpenAssetButton
   properties {toolTipText=Open asset}
   slot {class=HorizontalBoxSlot}
  }
  Image_89 {
   widgetClassPath=&Image
   bIsVariable=true
   parent=OpenAssetButton
   widget=Image_89
   properties {brush {
     bIsDynamicallyLoaded=false
     drawAs=Image
     imageType=FullColor
     mirroring=NoMirror
     resourceName=None
     resourceObject=/Game/NeverCook/EditorUtilities/DebugTools/AssetChangeFinder/Art/ArrowRight
     tiling=NoTile
     imageSize=<20 20>
     margin=<ltrb 0 0 0 0>
     outlineSettings {
      bUseBrushTransparency=false
      roundingType=HalfHeightRadius
      width=0
      color {colorUseRule=UseColor_Specified, specifiedColor=<rgba 0 0 0 0>}
      cornerRadii=<0 0 0 0>
     }
     tintColor {colorUseRule=UseColor_Specified, specifiedColor=<rgba 1 1 1 1>}
     uVRegion {bIsValid=false, max=<0 0>, min=<0 0>}
    }}
   slot {class=ButtonSlot}
  }
  BrowseToAssetButton {
   widgetClassPath=&EditorUtilityButton
   parent=HorizontalBox_30
   widget=BrowseToAssetButton
   properties {toolTipText=Browse to asset}
   slot {class=HorizontalBoxSlot}
  }
  Image_60 {
   widgetClassPath=&Image
   bIsVariable=true
   parent=BrowseToAssetButton
   widget=Image_60
   properties {brush {
     bIsDynamicallyLoaded=false
     drawAs=Image
     imageType=FullColor
     mirroring=NoMirror
     resourceName=None
     resourceObject=/Game/UI/Foundation/Textures/Icons/Replay/T-Icon-Replay-Folder-L
     tiling=NoTile
     imageSize=<20 20>
     margin=<ltrb 0 0 0 0>
     outlineSettings {
      bUseBrushTransparency=false
      roundingType=HalfHeightRadius
      width=0
      color {colorUseRule=UseColor_Specified, specifiedColor=<rgba 0 0 0 0>}
      cornerRadii=<0 0 0 0>
     }
     tintColor {colorUseRule=UseColor_Specified, specifiedColor=<rgba 1 1 1 1>}
     uVRegion {bIsValid=false, max=<0 0>, min=<0 0>}
    }}
   slot {class=ButtonSlot}
  }
 ]
}
