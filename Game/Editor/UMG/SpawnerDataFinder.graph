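# Editor utility widget on the FortSpawnerDataFinderTool C++ parent for searching spawner data assets by text and release version range.
# Clicking Search or committing the search text passes the text and combo box versions to the parent's FindAllSpawnerDataMatchingFilter.
# The max version equals the min unless the UseMaxVersion checkbox is ticked, which also drives a function for the max combo's enabled state.
# A CommonListView below displays entries as UMG_SpawnerDataListItem.
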EditorUtilityWidgetBlueprint /Game/Editor/UMG/SpawnerDataFinder
graphs [
 EventGraph {nodes [
   n1 = "On Clicked (SearchButton)"() {is=title K2Node_ComponentBoundEvent}
   n2 = Get SearchTextBox() -> (SearchTextBox:Editor Utility Editable Text Box Object Reference)
   n3 = GetSelectedOption!(self:"Combo Box (String) Object Reference"=Get MinReleaseVersionComboBox() -> (MinReleaseVersionComboBox:"Combo Box (String) Object Reference")) -> (ReturnValue:String)
   n4 = Get UseMaxVersion() -> (UseMaxVersion:Editor Utility Check Box Object Reference)
   n5 = Get MaxReleaseVersionComboBox() -> (MaxReleaseVersionComboBox:"Combo Box (String) Object Reference")
   FindAllSpawnerDataMatchingFilter!(
    execute=[@n1.then, @n6.then]
    self:Fort Spawner Data Finder Tool Object Reference
    MinReleaseVersion:String=@n3
    MaxReleaseVersion:String=Select(Option 0:String=@n3, Option 1:String=GetSelectedOption(self=@n5), Index=IsChecked(self=@n4)) -> (ReturnValue:String)
    SearchString:String="To String (Text)"(InText=Get Text(self=@n2))
   ) -> (then)
   n6 = "On Text Committed (SearchTextBox)"!() -> (OutputDelegate, then, Text:"Text (by ref)", CommitMethod:ETextCommit Enum=Default) {is=title K2Node_ComponentBoundEvent}
   Event PreConstruct()
   Event Construct()
   Event Tick()
  ]}
 Get_MaxReleaseVersionComboBox_bIsEnabled {nodes [
   Get_MaxReleaseVersionComboBox_bIsEnabled() {is=title K2Node_FunctionEntry}
   > Return Node!(execute, ReturnValue:bool=IsChecked(self=Get UseMaxVersion()))
  ]}
]
variables []
widgets {
 info {inheritedWidgetCount=3, parentClass=/Script/FortniteEditor.FortSpawnerDataFinderTool, rootWidgetClass=&VerticalBox, widgetCount=9}
 widgets [
  VerticalBox_18 {widgetClassPath=&VerticalBox, widget=VerticalBox_18}
  HorizontalBox_85 {widgetClassPath=&HorizontalBox, parent=VerticalBox_18, widget=HorizontalBox_85, slot {class=VerticalBoxSlot}}
  SearchTextBox {
   widgetClassPath=&EditorUtilityEditableTextBox
   bIsVariable=true
   parent=HorizontalBox_85
   widget=SearchTextBox
   slot {class=HorizontalBoxSlot, properties {size {sizeRule=Fill, value=1}}}
  }
  MinReleaseVersionComboBox {
   widgetClassPath=/Script/Blutility.EditorUtilityComboBoxString
   bInherited=true
   bIsVariable=true
   parent=HorizontalBox_85
   widget=MinReleaseVersionComboBox
   properties {toolTipText=Min Release Version, foregroundColor {colorUseRule=UseColor_Specified, specifiedColor=<rgba 0.5271 0.5271 0.5271 1>}}
   slot {class=HorizontalBoxSlot}
  }
  UseMaxVersion {
   widgetClassPath=&EditorUtilityCheckBox
   bIsVariable=true
   parent=HorizontalBox_85
   widget=UseMaxVersion
   properties {horizontalAlignment=HAlign_Left}
   slot {class=HorizontalBoxSlot}
  }
  MaxReleaseVersionComboBox {
   widgetClassPath=&EditorUtilityComboBoxString
   bInherited=true
   bIsVariable=true
   parent=HorizontalBox_85
   widget=MaxReleaseVersionComboBox
   properties {
    toolTipText=Max Release Version
    contentPadding=<ltrb 0 0 0 0>
    foregroundColor {colorUseRule=UseColor_Specified, specifiedColor=<rgba 0.5271 0.5271 0.5271 1>}
   }
   slot {class=HorizontalBoxSlot}
  }
  SearchButton {widgetClassPath=&EditorUtilityButton, parent=HorizontalBox_85, widget=SearchButton, slot {class=HorizontalBoxSlot}}
  TextBlock_89 {widgetClassPath=&TextBlock, parent=SearchButton, widget=TextBlock_89, properties {text=Search}, slot {class=ButtonSlot}}
  MatchingAssetListView {
   widgetClassPath=&CommonListView
   bInherited=true
   bIsVariable=true
   parent=VerticalBox_18
   widget=MatchingAssetListView
   properties {
    bEnableRightClickScrolling=false
    consumeMouseWheel=Always
    entryWidgetClass=&/UMG_SpawnerDataListItem.UMG_SpawnerDataListItem_C
    scrollIntoViewAlignment=IntoView
    selectionMode=None
    scrollBarStyle {
     thickness=8
     draggedThumbImage {
      bIsDynamicallyLoaded=false
      drawAs=RoundedBox
      imageType=NoImage
      mirroring=NoMirror
      resourceName=None
      resourceObject=null
      tiling=NoTile
      imageSize=<0 0>
      margin=<ltrb 0 0 0 0>
      outlineSettings {
       bUseBrushTransparency=false
       roundingType=FixedRadius
       width=0
       color {colorUseRule=UseColor_Specified, specifiedColor=<rgba 0 0 0 0>}
       cornerRadii=<4 4 4 4>
      }
      tintColor {colorUseRule=UseColor_Specified, specifiedColor=<rgba 0.2159 0.2159 0.2159 1>}
      uVRegion {bIsValid=false, max=<0 0>, min=<0 0>}
     }
     horizontalBackgroundImage {
      bIsDynamicallyLoaded=false
      drawAs=NoDrawType
      imageType=NoImage
      mirroring=NoMirror
      resourceName=None
      resourceObject=null
      tiling=NoTile
      imageSize=<0 0>
      margin=<ltrb 0 0 0 0>
      outlineSettings {
       bUseBrushTransparency=false
       roundingType=HalfHeightRadius
       width=0
       color {colorUseRule=UseColor_Specified, specifiedColor=<rgba 0 0 0 0>}
       cornerRadii=<0 0 0 0>
      }
      tintColor {colorUseRule=UseColor_Specified, specifiedColor=<rgba 1 1 1 1>}
      uVRegion {bIsValid=false, max=<0 0>, min=<0 0>}
     }
     horizontalBottomSlotImage {
      bIsDynamicallyLoaded=false
      drawAs=NoDrawType
      imageType=NoImage
      mirroring=NoMirror
      resourceName=None
      resourceObject=null
      tiling=NoTile
      imageSize=<0 0>
      margin=<ltrb 0 0 0 0>
      outlineSettings {
       bUseBrushTransparency=false
       roundingType=HalfHeightRadius
       width=0
       color {colorUseRule=UseColor_Specified, specifiedColor=<rgba 0 0 0 0>}
       cornerRadii=<0 0 0 0>
      }
      tintColor {colorUseRule=UseColor_Specified, specifiedColor=<rgba 1 1 1 1>}
      uVRegion {bIsValid=false, max=<0 0>, min=<0 0>}
     }
     horizontalTopSlotImage {
      bIsDynamicallyLoaded=false
      drawAs=NoDrawType
      imageType=NoImage
      mirroring=NoMirror
      resourceName=None
      resourceObject=null
      tiling=NoTile
      imageSize=<0 0>
      margin=<ltrb 0 0 0 0>
      outlineSettings {
       bUseBrushTransparency=false
       roundingType=HalfHeightRadius
       width=0
       color {colorUseRule=UseColor_Specified, specifiedColor=<rgba 0 0 0 0>}
       cornerRadii=<0 0 0 0>
      }
      tintColor {colorUseRule=UseColor_Specified, specifiedColor=<rgba 1 1 1 1>}
      uVRegion {bIsValid=false, max=<0 0>, min=<0 0>}
     }
     hoveredThumbImage {
      bIsDynamicallyLoaded=false
      drawAs=RoundedBox
      imageType=NoImage
      mirroring=NoMirror
      resourceName=None
      resourceObject=null
      tiling=NoTile
      imageSize=<0 0>
      margin=<ltrb 0 0 0 0>
      outlineSettings {
       bUseBrushTransparency=false
       roundingType=FixedRadius
       width=0
       color {colorUseRule=UseColor_Specified, specifiedColor=<rgba 0 0 0 0>}
       cornerRadii=<4 4 4 4>
      }
      tintColor {colorUseRule=UseColor_Specified, specifiedColor=<rgba 0.2159 0.2159 0.2159 1>}
      uVRegion {bIsValid=false, max=<0 0>, min=<0 0>}
     }
     normalThumbImage {
      bIsDynamicallyLoaded=false
      drawAs=RoundedBox
      imageType=NoImage
      mirroring=NoMirror
      resourceName=None
      resourceObject=null
      tiling=NoTile
      imageSize=<0 0>
      margin=<ltrb 0 0 0 0>
      outlineSettings {
       bUseBrushTransparency=false
       roundingType=FixedRadius
       width=0
       color {colorUseRule=UseColor_Specified, specifiedColor=<rgba 0 0 0 0>}
       cornerRadii=<4 4 4 4>
      }
      tintColor {colorUseRule=UseColor_Specified, specifiedColor=<rgba 0.0953 0.0953 0.0953 1>}
      uVRegion {bIsValid=false, max=<0 0>, min=<0 0>}
     }
     verticalBackgroundImage {
      bIsDynamicallyLoaded=false
      drawAs=NoDrawType
      imageType=NoImage
      mirroring=NoMirror
      resourceName=None
      resourceObject=null
      tiling=NoTile
      imageSize=<0 0>
      margin=<ltrb 0 0 0 0>
      outlineSettings {
       bUseBrushTransparency=false
       roundingType=HalfHeightRadius
       width=0
       color {colorUseRule=UseColor_Specified, specifiedColor=<rgba 0 0 0 0>}
       cornerRadii=<0 0 0 0>
      }
      tintColor {colorUseRule=UseColor_Specified, specifiedColor=<rgba 1 1 1 1>}
      uVRegion {bIsValid=false, max=<0 0>, min=<0 0>}
     }
     verticalBottomSlotImage {
      bIsDynamicallyLoaded=false
      drawAs=NoDrawType
      imageType=NoImage
      mirroring=NoMirror
      resourceName=None
      resourceObject=null
      tiling=NoTile
      imageSize=<0 0>
      margin=<ltrb 0 0 0 0>
      outlineSettings {
       bUseBrushTransparency=false
       roundingType=HalfHeightRadius
       width=0
       color {colorUseRule=UseColor_Specified, specifiedColor=<rgba 0 0 0 0>}
       cornerRadii=<0 0 0 0>
      }
      tintColor {colorUseRule=UseColor_Specified, specifiedColor=<rgba 1 1 1 1>}
      uVRegion {bIsValid=false, max=<0 0>, min=<0 0>}
     }
     verticalTopSlotImage {
      bIsDynamicallyLoaded=false
      drawAs=NoDrawType
      imageType=NoImage
      mirroring=NoMirror
      resourceName=None
      resourceObject=null
      tiling=NoTile
      imageSize=<0 0>
      margin=<ltrb 0 0 0 0>
      outlineSettings {
       bUseBrushTransparency=false
       roundingType=HalfHeightRadius
       width=0
       color {colorUseRule=UseColor_Specified, specifiedColor=<rgba 0 0 0 0>}
       cornerRadii=<0 0 0 0>
      }
      tintColor {colorUseRule=UseColor_Specified, specifiedColor=<rgba 1 1 1 1>}
      uVRegion {bIsValid=false, max=<0 0>, min=<0 0>}
     }
    }
   }
   slot {class=VerticalBoxSlot, properties {size {sizeRule=Fill, value=1}}}
  }
 ]
}
